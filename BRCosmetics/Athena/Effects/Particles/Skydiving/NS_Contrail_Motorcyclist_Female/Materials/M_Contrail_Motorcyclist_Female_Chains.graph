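# Masked mesh-particle material for swinging chains with weapon tips in the female Motorcyclist skydiving contrail.
# World position offset rotates vertices about a wandering axis with height-lagged sine motion, then widens chain strips toward camera-facing ribbons.
# A chainlink texture, slid in stepped sine-driven increments, supplies opacity and rebuilt view-facing normals, while UV1 excludes the tips.
# Distance from particle position fades base color and specular into emissive glow, tuned by Dynamic Parameters.
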Material /BRCosmetics/Athena/Effects/Particles/Skydiving/NS_Contrail_Motorcyclist_Female/Materials/M_Contrail_Motorcyclist_Female_Chains
graph {
 properties {
  bTangentSpaceNormal=false
  bUsedWithNiagaraMeshParticles=true
  blendMode=BLEND_Masked
  pixelDepthOffsetMode=PDOM_Legacy
  twoSided=true
  usedShadingModels=Default Lit
 }
 nodes [
  n1 = WorldPosition()
  n2 = Normalize(VectorInput=Transform(None=Constant3Vector() {desc=rotation axis' rotation axis, constant=<rgba 1 0 0 0>}) {transformSourceType=TRANSFORMSOURCE_Local})
  n3 = DynamicParameter() {defaultValue=<rgba 1 5 15 22>, paramNames [Time Offset, Chain Width, Sphere Mask Offset, Sphere Mask Radius]}
  n4 = Time()
  n5 = ParticleRandom()
  n6 = Add(A=@n3.Time Offset, B=Multiply(A=Add(A=@n4, B=@n5), B=ScalarParameter() {defaultValue=0.618, parameterName=Axis Rotation Speed}))
  n7 = Constant()
  n8 = Normalize(VectorInput=Transform(None=Constant3Vector() {desc=rotation axis, constant=<rgba 0 1 0 0>}) {transformSourceType=TRANSFORMSOURCE_Local})
  n9 = Normalize(VectorInput=Add(A=RotateAboutAxis(NormalizedRotationAxis=@n2, RotationAngle=@n6, PivotPoint=@n7, Position=@n8), B=@n8))
  n10 = Add(A=Multiply(A=Time(), B=ScalarParameter() {defaultValue=2.618, parameterName=Rotation Speed}), B=@n3.Time Offset)
  n11 = TextureCoordinate()
  n12 = ComponentMask(None=TextureCoordinate() {coordinateIndex=1}) {r=true}
  n13 = Multiply(A=Max(A=ComponentMask(None=@n11) {r=true}, B=OneMinus(None=@n12)), B=ScalarParameter() {defaultValue=0.86, parameterName=Height Time Offset})
  n14 = Multiply(A=Sine(None=Subtract(A=@n10, B=@n13)), B=ScalarParameter() {defaultValue=0.025, parameterName=Rotation Angle Max})
  n15 = RotateAboutAxis(
   NormalizedRotationAxis=@n9
   RotationAngle=@n14
   PivotPoint=MaterialFunctionCall() {materialFunction=&ObjectPivotPoint}.Mesh Particle Pivot Location
   Position=WorldPosition().XYZ
  )
  n16 = Normalize(VectorInput=Subtract(A=CameraPositionWS(), B=Add(A=@n1.XYZ, B=@n15)))
  n17 = Transform(None=Constant3Vector() {constant=<rgba 1 0 0 0>})
  n18 = Normalize(VectorInput=CrossProduct(A=@n16, B=Normalize(VectorInput=@n17)))
  n19 = TextureCoordinate()
  n20 = Multiply(A=@n18, B=Subtract(A=ComponentMask(None=@n19) {g=true}) {constB=0.5})
  n21 = DynamicParameter() {defaultValue=<rgba 1 5 15 22>, paramNames [Time Offset, Chain Width, Sphere Mask Offset, Sphere Mask Radius]}
  n22 = ComponentMask(None=TextureCoordinate() {coordinateIndex=1}) {r=true}
  Add(A=Multiply(A=Multiply(A=@n20, B=@n21.Chain Width), B=@n22) {desc=Masking out the spline thicken effect at the weapon tips}, B=@n15)
  Constant() {r=1}
  n23 = VertexNormalWS()
  n24 = RotateAboutAxis(NormalizedRotationAxis=@n9, RotationAngle=@n14, PivotPoint=Constant(), Position=@n23)
  n25 = TextureCoordinate()
  n26 = AppendVector(A=ScalarParameter() {defaultValue=1.5, parameterName=UV Scale}, B=Constant() {r=1})
  n27 = Multiply(A=Time(), B=ScalarParameter() {defaultValue=0.333, parameterName=Panner Speed})
  n28 = Multiply(A=@n27) {constB=2.618}
  n29 = Multiply(A=Multiply(A=Sine(None=@n27), B=Sine(None=@n28)), B=Sine(None=Multiply(A=@n27) {constB=4.235}))
  n30 = Divide(A=Floor(None=Multiply(A=@n29) {constB=8})) {constB=8}
  n31 = Constant()
  n32 = TextureSample(UVs=Add(A=Multiply(A=@n25, B=@n26), B=AppendVector(A=@n30, B=@n31))) {
   samplerType=SAMPLERTYPE_Masks
   texture=/BRCosmetics/Athena/Effects/Particles/Skydiving/NS_Contrail_Motorcyclist_Female/Textures/T_Contrail_Motorcyclist_Female_Chainlink
  }
  n33 = DeriveNormalZ(InXY=ConstantBiasScale(None=ComponentMask(None=@n32.RGBA) {g=true, r=true}) {bias=-0.5, scale=2})
  n34 = Multiply(A=VertexInterpolator(VS=CrossProduct(A=@n18, B=@n16)).PS, B=ComponentMask(None=@n33) {r=true})
  n35 = Multiply(A=VertexInterpolator(VS=@n18).PS, B=ComponentMask(None=@n33) {g=true})
  n36 = VertexInterpolator(VS=@n16)
  n37 = ComponentMask(None=@n33) {b=true}
  LinearInterpolate(A=Normalize(VectorInput=Add(A=@n24, B=@n23)), B=Add(A=Add(A=@n34, B=@n35), B=Multiply(A=@n36.PS, B=@n37)), Alpha=@n22) {desc=Lerping other set of normals for the weapon tips}
  n38 = WorldPosition()
  n39 = ParticlePositionWS()
  n40 = DynamicParameter() {defaultValue=<rgba 1 5 15 22>, paramNames [Time Offset, Chain Width, Sphere Mask Offset, Sphere Mask Radius]}
  n41 = Divide(A=Subtract(A=Distance(A=@n38.XYZ, B=@n39), B=@n40.Sphere Mask Offset), B=@n40.Sphere Mask Radius)
  n42 = VectorParameter() {parameterName=Glow Color, defaultValue=<rgba 0 0.9667 1 3>}
  n43 = Power(Base=OneMinus(None=Saturate(None=@n41)), Exp=@n42.A)
  MaterialFunctionCall(Emissive=Multiply(A=@n43, B=@n42.RGB)) {materialFunction=&MF_FX_EmissiveAdjustments}
  n44 = OneMinus(None=ComponentMask(None=TextureCoordinate() {coordinateIndex=1}) {r=true})
  Max(A=@n44, B=Round(None=ComponentMask(None=@n32.RGBA) {a=true})) {desc=100% opacity for weapon tips}
  n45 = ComponentMask(None=TextureCoordinate() {coordinateIndex=1}) {g=true}
  n46 = Multiply(A=Floor(None=Multiply(A=@n45) {constB=100})) {constB=0.5}
  Multiply(A=ComponentMask(None=VectorNoise(World Position=@n46)) {r=true}, B=@n3.Time Offset)
  n47 = OneMinus(None=@n43)
  n48 = ComponentMask(None=@n32.RGBA) {b=true}
  Multiply(
   A=@n47
   B=Multiply(
    A=Max(A=@n48, B=@n44) {desc=Hiding texture for weapon tips}
    B=VectorParameter() {parameterName=Base Color, defaultValue=<rgba 0.01 0.01 0.01 0.55>}.RGB
   )
  )
  Multiply(A=@n47, B=ScalarParameter() {defaultValue=1, parameterName=Specular})
  ScalarParameter() {defaultValue=15, parameterName=Sphere Mask Offset}
  ScalarParameter() {defaultValue=22, parameterName=Sphere Mask radius}
  ScalarParameter() {defaultValue=0.45, parameterName=Time Offset Amount}
  SphereMask(A=@n38.XYZ, B=@n39, Radius=@n40.Sphere Mask Radius) {hardnessPercent=0}
 ]
}
