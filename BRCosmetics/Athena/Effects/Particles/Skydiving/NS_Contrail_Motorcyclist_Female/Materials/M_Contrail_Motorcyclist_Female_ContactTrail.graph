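# Additive unlit Niagara ribbon material for the contact trail of the female Motorcyclist skydiving contrail.
# Two panning samples of the wheel texture on scaled UVs multiply into a tire tread mask.
# That mask is screen-blended with its eroded version and tinted by particle color into an emissive adjustment function.
# A fade term combines a reflected gradient across V, particle alpha and a capped view-angle falloff.
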Material /BRCosmetics/Athena/Effects/Particles/Skydiving/NS_Contrail_Motorcyclist_Female/Materials/M_Contrail_Motorcyclist_Female_ContactTrail
graph {
 properties {bUsedWithNiagaraRibbons=true, blendMode=BLEND_Additive, pixelDepthOffsetMode=PDOM_Legacy, shadingModel=MSM_Unlit, usedShadingModels=Unlit}
 nodes [
  n1 = MaterialFunctionCall(Float2=TextureCoordinate()) {materialFunction=&BreakOutFloat2Components}
  n2 = ScalarParameter() {defaultValue=4, parameterName=UV Scale U}
  n3 = AppendVector(
   A=Power(Base=Multiply(A=@n1.R, B=@n2), Exp=ScalarParameter() {defaultValue=1, parameterName=UV Exponent U})
   B=Multiply(A=@n1.G, B=ScalarParameter() {defaultValue=1.1, parameterName=UV Scale V})
  )
  n4 = ScalarParameter() {defaultValue=-3.45, parameterName=Tire Tread Overlay Panner Speed A X}
  n5 = ScalarParameter() {defaultValue=0.618, parameterName=Tire Tread Overlay Panner Speed A Y}
  n6 = TextureSample(UVs=Panner(Coordinate=@n3, Speed=AppendVector(A=@n4, B=@n5)) {bFractionalPart=true}) {
   samplerType=SAMPLERTYPE_Masks
   texture=/BRCosmetics/Athena/Effects/Particles/Skydiving/NS_Contrail_Motorcyclist_Female/Textures/T_Contrail_Motorcyclist_Female_Wheel
  }
  n7 = ScalarParameter() {defaultValue=-4.23, parameterName=Tire Tread Overlay Panner Speed B X}
  n8 = ScalarParameter() {defaultValue=-0.45, parameterName=Tire Tread Overlay Panner Speed B Y}
  n9 = TextureSample(UVs=Panner(Coordinate=@n3, Speed=AppendVector(A=@n7, B=@n8)) {bFractionalPart=true}) {
   samplerType=SAMPLERTYPE_Masks
   texture=&T_Contrail_Motorcyclist_Female_Wheel
  }
  n10 = Power(Base=Multiply(A=@n6.G, B=@n9.G), Exp=ScalarParameter() {defaultValue=2, parameterName=Tire Tread Mask Exponent})
  n11 = MaterialFunctionCall!(
   Gradient=@n10
   Erosion Amount=ScalarParameter() {defaultValue=0.25, parameterName=Erosion}
   Softness=ScalarParameter() {defaultValue=0.1, parameterName=Softness}
  ) {materialFunction=/BRCosmetics/Athena/Effects/Particles/Emotes/NS_Emote_Spectrum/Functions/MF_Erosion}
  n12 = Multiply(A=OneMinus(None=@n10), B=OneMinus(None=@n11.Default Range))
  n13 = ParticleColor()
  MaterialFunctionCall(Emissive=Multiply(A=OneMinus(None=@n12), B=@n13.RGB)) {materialFunction=&MF_FX_EmissiveAdjustments}
  Max()
  n14 = MaterialFunctionCall!(Gradient=@n1.G, Center Point=ScalarParameter() {defaultValue=0.9, parameterName=Center Point}) {materialFunction=/Game/Effects/Fort_Effects/MaterialFunctions/MF_GradientReflected}
  Multiply(A=@n14.Result, B=@n14.Result)
  n15 = VertexNormalWS()
  n16 = CameraVectorWS()
  n17 = SquareRoot(None=Abs(None=DotProduct(A=@n15, B=@n16)))
  Multiply(
   A=Power(Base=@n14.Result, Exp=ScalarParameter() {defaultValue=1, parameterName=Gradient Fade Exponent})
   B=Multiply(A=@n13.A, B=Min(A=@n17) {constB=0.333})
  )
  Saturate(None=Add(A=@n10, B=@n11.Default Range))
 ]
}
