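# Unlit translucent Niagara ribbon material that draws the road trail of the female Motorcyclist skydiving contrail.
# A panning trail texture supplies erosion-shaped opacity and lerps pavement and lane-line gradients along U.
# World Z relative to the Initial Z Position dynamic parameter drives fades and a glowing contact surface, eroded from two multiplied panning tire-tread samples.
# Opacity also fades by facing angle, ribbon length and scene depth; several leftover branches feed nothing.
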Material /BRCosmetics/Athena/Effects/Particles/Skydiving/NS_Contrail_Motorcyclist_Female/Materials/M_Contrail_Motorcyclist_Female_Road
graph {
 properties {bUsedWithNiagaraRibbons=true, blendMode=BLEND_Translucent, pixelDepthOffsetMode=PDOM_Legacy, shadingModel=MSM_Unlit, usedShadingModels=Unlit}
 nodes [
  n1 = DynamicParameter() {paramNames [Initial Z Position, "", "", ""]}
  n2 = Subtract(A=ComponentMask(None=WorldPosition().XYZ) {b=true}, B=@n1.Initial Z Position)
  Abs(None=@n2)
  n3 = DotProduct(A=CameraVectorWS(), B=VertexNormalWS())
  n4 = Abs(None=@n3)
  n5 = TextureCoordinate()
  n6 = AppendVector(A=ScalarParameter() {defaultValue=1.5, parameterName=UV U Scale}, B=Constant() {r=1})
  n7 = ScalarParameter() {defaultValue=-2, parameterName=UV U Speed}
  n8 = Constant()
  n9 = TextureSample(UVs=Panner(Coordinate=Multiply(A=@n5, B=@n6), Speed=AppendVector(A=@n7, B=@n8)) {bFractionalPart=true}) {
   samplerType=SAMPLERTYPE_Masks
   texture=/BRCosmetics/Athena/Effects/Particles/Skydiving/NS_Contrail_Motorcyclist_Female/Textures/T_Contrail_Motorcyclist_Female_Trail
  }
  n10 = ComponentMask(None=TextureCoordinate()) {g=true}
  n11 = ScalarParameter() {defaultValue=2, parameterName=Erosion Exponent}
  n12 = Power(Base=Abs(None=ConstantBiasScale(None=@n10) {bias=-0.5, scale=2}), Exp=@n11) {constExponent=4}
  n13 = Divide(A=@n2, B=ScalarParameter() {parameterName=Initial Z Falloff})
  n14 = Saturate(None=@n13)
  n15 = Multiply(A=@n14, B=@n14)
  n16 = OneMinus(None=Power(Base=OneMinus(None=@n15), Exp=@n11) {constExponent=4})
  n17 = OneMinus(None=Multiply(A=OneMinus(None=@n12), B=@n16))
  n18 = MaterialFunctionCall!(Gradient=@n9.R, Erosion Amount=SquareRoot(None=@n17), Softness=ScalarParameter() {defaultValue=0.4, parameterName=Softness}) {materialFunction=/BRCosmetics/Athena/Effects/Particles/Skydiving/NS_Contrail_DarkAndLight/Functions/MF_Erosion}
  n19 = ComponentMask(None=TextureCoordinate()) {r=true}
  n20 = Multiply(A=@n16, B=OneMinus(None=@n19))
  DepthFade(
   Opacity=Multiply(A=Power(Base=@n4) {constExponent=0.333}, B=Multiply(A=@n18.Increased Range, B=@n20))
   FadeDistance=ScalarParameter() {defaultValue=9, parameterName=Depth Fade Distance}
  )
  Frac(None=@n13)
  n21 = ComponentMask(None=TextureCoordinate()) {r=true}
  n22 = ScalarParameter() {defaultValue=0.1, parameterName=Contact Surface Shift, sliderMax=1}
  n23 = OneMinus(None=Frac(None=Subtract(A=@n21, B=@n22)))
  n24 = ScalarParameter() {defaultValue=0.4, parameterName=Contact Surface Length}
  Frac(None=Saturate(None=Divide(A=@n23, B=@n24)))
  MaterialFunctionCall!(Gradient=@n19, Center Point=Constant() {r=0.2}) {materialFunction=/Game/Effects/Fort_Effects/MaterialFunctions/MF_GradientReflected}
  MaterialFunctionCall!(Gradient=@n9.R, Erosion Amount=@n17) {materialFunction=/BRCosmetics/Athena/Effects/Particles/Skydiving/NS_Contrail_DarkAndLight/Functions/MF_ErosionNormalized}
  MaterialFunctionCall(Number=@n1.Initial Z Position) {materialFunction=&DebugScalarValues}
  n25 = VectorParameter() {parameterName=Contact Surface, defaultValue=<rgba 0 6 6 0>}
  n26 = Power(Base=NamedRerouteUsage(), Exp=ScalarParameter() {defaultValue=0.333, parameterName=Contact Surface Mask Exponent})
  n27 = Saturate(None=Divide(A=@n2, B=ScalarParameter() {defaultValue=66, parameterName=Streaks Z Falloff}))
  n28 = Saturate(None=Power(
    Base=MaterialFunctionCall(Gradient=@n26, Erosion Amount=@n27) {materialFunction=&MF_ErosionNormalized}.Result
    Exp=ScalarParameter() {defaultValue=3, parameterName=Contact Surface Final Exponent}
   ))
  n29 = VectorParameter() {parameterName=Pavement Color A, defaultValue=<rgba 0 1 1 1>}
  n30 = VectorParameter() {parameterName=Lines Color A, defaultValue=<rgba 0.9 1 1 0.45>}
  n31 = Power(Base=ComponentMask(None=TextureCoordinate()) {r=true}, Exp=@n30.A)
  n32 = LinearInterpolate(
   A=LinearInterpolate(A=@n29.RGB, B=VectorParameter() {parameterName=Pavement Color B, defaultValue=<rgba 0.0833 0 1 0>}.RGB, Alpha=@n31)
   B=LinearInterpolate(A=@n30.RGB, B=VectorParameter() {parameterName=Lines Color B, defaultValue=<rgba 0 3 3 0>}.RGB, Alpha=@n31)
   Alpha=@n9.G
  )
  MaterialFunctionCall(Emissive=Multiply(A=LinearInterpolate(A=@n32, B=@n25.RGB, Alpha=@n28), B=ParticleColor().A)) {materialFunction=&MF_FX_EmissiveAdjustments}
  n33 = TwoSidedSign()
  Multiply(A=OneMinus(None=Saturate(None=@n33)), B=@n28)
  Multiply(A=@n3, B=@n3)
  Multiply(A=@n18.Increased Range, B=@n29.A)
  Multiply(A=@n20, B=@n20)
  OneMinus(None=@n27)
  n34 = ComponentMask(None=TextureCoordinate()) {g=true}
  n35 = ScalarParameter() {defaultValue=0.3, parameterName=Contact Surface Mask Size}
  n36 = Power(Base=Saturate(None=Divide(A=@n34, B=@n35)), Exp=ScalarParameter() {defaultValue=3, parameterName=Contact Surface Width Exponent})
  n37 = MaterialFunctionCall(Gradient=Frac(None=@n27), Center Point=ScalarParameter() {defaultValue=0.2, parameterName=Contact Surface Centerpoint, sliderMax=1}) {materialFunction=&MF_GradientReflected}
  n38 = ScalarParameter() {defaultValue=2, parameterName=Contact Surface Length Exponent}
  OneMinus(None=Multiply(A=OneMinus(None=@n36), B=Power(Base=@n37.Result, Exp=@n38)))
  n39 = AppendVector(A=Divide(A=@n2, B=ScalarParameter() {defaultValue=-66, parameterName=UV Z Scale}), B=ComponentMask(None=TextureCoordinate()) {g=true})
  n40 = ScalarParameter() {defaultValue=2.618, parameterName=Tire Tread Overlay Panner Speed A X}
  n41 = ScalarParameter() {defaultValue=0.618, parameterName=Tire Tread Overlay Panner Speed A Y}
  n42 = TextureSample(UVs=Panner(Coordinate=@n39, Speed=AppendVector(A=@n40, B=@n41)) {bFractionalPart=true}) {
   samplerType=SAMPLERTYPE_Masks
   texture=/BRCosmetics/Athena/Effects/Particles/Skydiving/NS_Contrail_Motorcyclist_Female/Textures/T_Contrail_Motorcyclist_Female_Wheel
  }
  n43 = ScalarParameter() {defaultValue=4.23, parameterName=Tire Tread Overlay Panner Speed B X}
  n44 = ScalarParameter() {defaultValue=-0.45, parameterName=Tire Tread Overlay Panner Speed B Y}
  n45 = TextureSample(UVs=Panner(Coordinate=@n39, Speed=AppendVector(A=@n43, B=@n44)) {bFractionalPart=true}) {
   samplerType=SAMPLERTYPE_Masks
   texture=&T_Contrail_Motorcyclist_Female_Wheel
  }
  Power(
   Base=NamedRerouteDeclaration(None=Multiply(A=@n42.G, B=@n45.G)) {name=Directional Noise, nodeColor=<rgba 0.4297 1 0 1>}
   Exp=ScalarParameter() {defaultValue=2, parameterName=Tire Tread Mask Exponent}
  )
  SquareRoot(None=@n4)
 ]
}
