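# Additive, unlit, two-sided material for the smear layer of the female Motorcyclist skydiving contrail, built for Niagara ribbons and mesh particles.
# Two panners scroll a wheel mask over warped UVs, offset by face side and a dynamic parameter, forming noise.
# MF_Erosion cuts this noise using tip, side, particle-alpha and non-mobile depth fades, then squared view-facing softens silhouettes.
# A two-color gradient along U drives emissive, and a Push parameter scales vertex normals.
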Material /BRCosmetics/Athena/Effects/Particles/Skydiving/NS_Contrail_Motorcyclist_Female/Materials/M_Contrail_Motorcyclist_Female_Smear
graph {
 properties {
  bUsedWithNiagaraMeshParticles=true
  bUsedWithNiagaraRibbons=true
  blendMode=BLEND_Additive
  pixelDepthOffsetMode=PDOM_Legacy
  shadingModel=MSM_Unlit
  twoSided=true
  usedShadingModels=Unlit
 }
 nodes [
  n1 = VectorParameter() {parameterName=Smear Color A, defaultValue=<rgba 0.0274 0.021 0.03 1>}
  n2 = ComponentMask(None=TextureCoordinate()) {r=true}
  MaterialFunctionCall(Emissive=LinearInterpolate(
    A=@n1.RGB
    B=VectorParameter() {parameterName=Smear Color B, defaultValue=<rgba 0.2444 0.0233 0.333 0>}.RGB
    Alpha=Power(Base=@n2, Exp=@n1.A)
   )) {materialFunction=&MF_FX_EmissiveAdjustments}
  n3 = MaterialFunctionCall(Float2=TextureCoordinate()) {materialFunction=&BreakOutFloat2Components}
  n4 = OneMinus(None=Power(Base=@n3.R, Exp=ScalarParameter() {defaultValue=3, parameterName=UV Exponent U}))
  n5 = ScalarParameter() {defaultValue=3, parameterName=UV Scale U}
  n6 = ScalarParameter() {defaultValue=1, parameterName=UV Scale V}
  n7 = TwoSidedSign()
  n8 = Add(A=Multiply(A=AppendVector(A=@n4, B=@n3.G), B=AppendVector(A=@n5, B=@n6)), B=Divide(A=Saturate(None=@n7)))
  n9 = Add(A=Time(), B=DynamicParameter() {paramNames ["", Texture Offset, "", ""]}.Texture Offset)
  n10 = ScalarParameter() {defaultValue=2.618, parameterName=Tire Tread Overlay Panner Speed A X}
  n11 = ScalarParameter() {defaultValue=-0.123, parameterName=Tire Tread Overlay Panner Speed A Y}
  n12 = TextureSample(UVs=Panner(Coordinate=@n8, Time=@n9, Speed=AppendVector(A=@n10, B=@n11)) {bFractionalPart=true}) {
   samplerType=SAMPLERTYPE_Masks
   texture=/BRCosmetics/Athena/Effects/Particles/Skydiving/NS_Contrail_Motorcyclist_Female/Textures/T_Contrail_Motorcyclist_Female_Wheel
  }
  n13 = ScalarParameter() {defaultValue=4.23, parameterName=Tire Tread Overlay Panner Speed B X}
  n14 = ScalarParameter() {defaultValue=0.2, parameterName=Tire Tread Overlay Panner Speed B Y}
  n15 = TextureSample(UVs=Panner(Coordinate=@n8, Time=@n9, Speed=AppendVector(A=@n13, B=@n14)) {bFractionalPart=true}) {
   samplerType=SAMPLERTYPE_Masks
   texture=&T_Contrail_Motorcyclist_Female_Wheel
  }
  n16 = Power(Base=Divide(A=Add(A=@n12.G, B=@n15.G)), Exp=ScalarParameter() {defaultValue=1, parameterName=Noise Exponent})
  n17 = TextureCoordinate()
  n18 = Divide(A=Subtract(A=ComponentMask(None=@n17) {r=true}) {constB=0.1}, B=ScalarParameter() {defaultValue=0.5, parameterName=Erosion Size Tip})
  n19 = ComponentMask(None=@n17) {g=true}
  n20 = Power(
   Base=Abs(None=ConstantBiasScale(None=@n19) {bias=-0.5, scale=2})
   Exp=ScalarParameter() {defaultValue=3, parameterName=Side Erosion Exponent}
  ) {constExponent=6}
  n21 = Multiply(A=Multiply(A=Saturate(None=@n18), B=OneMinus(None=@n20)), B=ParticleColor().A)
  n22 = ScalarParameter() {defaultValue=33, parameterName=Depth Fade Distance}
  n23 = OneMinus(None=ShadingPathSwitch(Default=DepthFade(Opacity=@n21, FadeDistance=@n22), Mobile=@n21) {desc=Migrated from FeatureLevelSwitch})
  n24 = ScalarParameter() {defaultValue=1, parameterName=soft}
  n25 = Abs(None=DotProduct(A=VertexNormalWS(), B=CameraVectorWS()))
  Multiply(
   A=ParticleColor().A
   B=Multiply(
    A=MaterialFunctionCall!(Gradient=@n16, Erosion Amount=@n23, Softness=@n24) {materialFunction=/BRCosmetics/Athena/Effects/Particles/Skydiving/NS_Contrail_DarkAndLight/Functions/MF_Erosion}.Default Range
    B=Multiply(A=@n25, B=@n25)
   )
  )
  Multiply(A=VertexNormalWS(), B=ScalarParameter() {defaultValue=-0.2, parameterName=Push})
 ]
}
