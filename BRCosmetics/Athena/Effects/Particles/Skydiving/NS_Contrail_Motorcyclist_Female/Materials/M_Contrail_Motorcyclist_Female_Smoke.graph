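# Translucent, unlit, two-sided Niagara ribbon material for the smoke in the female Motorcyclist skydiving contrail cosmetic.
# A panning trail mask, offset by a dynamic parameter, is eroded toward the ribbon's sides and ends using UV gradients and a normalized erosion function.
# The result is faded along U, then depth-faded except on Low and Medium quality.
# That mask blends particle color in lerp nodes, one labeled as an old overlay-blend variant.
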Material /BRCosmetics/Athena/Effects/Particles/Skydiving/NS_Contrail_Motorcyclist_Female/Materials/M_Contrail_Motorcyclist_Female_Smoke
graph {
 properties {
  bAllowNegativeEmissiveColor=true
  bUsedWithNiagaraRibbons=true
  blendMode=BLEND_Translucent
  pixelDepthOffsetMode=PDOM_Legacy
  shadingModel=MSM_Unlit
  twoSided=true
  usedShadingModels=Unlit
 }
 nodes [
  n1 = ParticleColor()
  n2 = Multiply(A=TextureCoordinate(), B=AppendVector(A=ScalarParameter() {defaultValue=0.5, parameterName=UV U Scale}, B=Constant() {r=1}))
  n3 = AppendVector(A=ScalarParameter() {defaultValue=-1, parameterName=UV U Speed}, B=Constant())
  n4 = DynamicParameter() {paramNames [- unused -, Texture Offset, - unused -, - unused -]}
  n5 = Constant()
  n6 = TextureSample(UVs=Add(A=Panner(Coordinate=@n2, Speed=@n3) {bFractionalPart=true}, B=AppendVector(A=@n4.Texture Offset, B=@n5)) {constB=0}) {
   samplerType=SAMPLERTYPE_Masks
   texture=/BRCosmetics/Athena/Effects/Particles/Skydiving/NS_Contrail_Motorcyclist_Female/Textures/T_Contrail_Motorcyclist_Female_Trail
  }
  n7 = ComponentMask(None=TextureCoordinate()) {g=true}
  n8 = ScalarParameter() {defaultValue=4, parameterName=Erosion Exponent}
  n9 = Power(Base=Abs(None=ConstantBiasScale(None=@n7) {bias=-0.5, scale=2}), Exp=@n8) {constExponent=4}
  n10 = ComponentMask(None=TextureCoordinate()) {r=true}
  n11 = MaterialFunctionCall!(Gradient=@n10, Center Point=Constant() {r=0.2}) {materialFunction=/Game/Effects/Fort_Effects/MaterialFunctions/MF_GradientReflected}
  n12 = OneMinus(None=Power(Base=OneMinus(None=@n11.Result), Exp=@n8) {constExponent=4})
  n13 = OneMinus(None=Multiply(A=OneMinus(None=@n9), B=@n12))
  n14 = OneMinus(None=@n10)
  n15 = Multiply(
   A=MaterialFunctionCall!(Gradient=@n6.B, Erosion Amount=Saturate(None=@n13)) {materialFunction=/BRCosmetics/Athena/Effects/Particles/Skydiving/NS_Contrail_DarkAndLight/Functions/MF_ErosionNormalized}.Result
   B=Saturate(None=Multiply(A=@n12, B=@n14))
  )
  n16 = DepthFade(Opacity=@n15, FadeDistance=ScalarParameter() {defaultValue=11, parameterName=Depth Fade Distance})
  n17 = Saturate(None=QualitySwitch(Default=@n16, Low=@n15, High=@n16, Medium=@n15, Epic=@n16))
  LinearInterpolate(A=@n1.RGB, Alpha=@n17)
  LinearInterpolate(B=@n1.RGB, Alpha=@n17) {constA=1, desc="old, for overlay blend"}
  Multiply(A=@n1.RGB) {constB=-1}
 ]
}
