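# Unlit, two-sided Niagara mesh material for the glowing tire in the female Motorcyclist skydiving contrail.
# Panning tread, axis-aligned Fresnel side coloring and noise streaks form the outer surface, while TwoSidedSign selects an inside color.
# An eroded contact patch, center glow, rim light and screen-blended, eroded comet noise layer on top.
# The result gets a UV-gradient hue shift and particle-alpha scaling before emissive adjustment, with one packed texture supplying all masks.
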Material /BRCosmetics/Athena/Effects/Particles/Skydiving/NS_Contrail_Motorcyclist_Female/Materials/M_Contrail_Motorcyclist_Female_Tire
graph {
 properties {bUsedWithNiagaraMeshParticles=true, pixelDepthOffsetMode=PDOM_Legacy, shadingModel=MSM_Unlit, twoSided=true, usedShadingModels=Unlit}
 nodes [
  n1 = Subtract(A=ComponentMask(None=TextureCoordinate() {coordinateIndex=2}) {g=true}, B=ScalarParameter() {parameterName=Hue Shift Gradient Shift})
  n2 = ScalarParameter() {defaultValue=0.5, parameterName=Hue Shift Gradient Length}
  n3 = Multiply(A=Saturate(None=Divide(A=@n1, B=@n2)), B=ScalarParameter() {defaultValue=0.1, parameterName=Hue Shift Intensity})
  n4 = OneMinus(None=DotProduct(A=VertexNormalWS(), B=CameraVectorWS()))
  n5 = VectorParameter() {parameterName=Rim Lighting, defaultValue=<rgba 0 0.25 0.1875 5>}
  n6 = Add(A=Time(), B=ParticleRandom())
  n7 = ScalarParameter() {defaultValue=3.1, parameterName=Tire Tread Panner Speed}
  n8 = Constant()
  n9 = TextureSample(UVs=Panner(Time=@n6, Speed=AppendVector(A=@n7, B=@n8)) {bFractionalPart=true}) {
   samplerType=SAMPLERTYPE_Masks
   texture=/BRCosmetics/Athena/Effects/Particles/Skydiving/NS_Contrail_Motorcyclist_Female/Textures/T_Contrail_Motorcyclist_Female_Wheel
  }
  n10 = OneMinus(None=@n9.R)
  n11 = ScalarParameter() {defaultValue=0.4, parameterName=Tire Tread Intensity}
  n12 = VertexNormalWS()
  n13 = Transform(None=Constant3Vector() {constant=<rgba 0 0 1 0>}) {transformSourceType=TRANSFORMSOURCE_Instance}
  n14 = DotProduct(A=@n12, B=@n13)
  n15 = OneMinus(None=Power(Base=Abs(None=@n14), Exp=ScalarParameter() {defaultValue=1, parameterName=Side Blend Exponent}))
  n16 = MaterialFunctionCall!(Axis=@n13, Exponent=ScalarParameter() {defaultValue=2, parameterName=Axis Aligned Fresnel Exponent}, Normal=@n12) {materialFunction=/Game/Effects/Fort_Effects/Materials/_Functions/MF_AxisAlignedFrensnel}
  n17 = Multiply(
   A=OneMinus(None=Multiply(A=@n10, B=@n11))
   B=LinearInterpolate(
    A=VectorParameter() {parameterName=Color Sides, defaultValue=<rgba 0 0.061 0.1 0>}.RGB
    B=VectorParameter() {parameterName=Color Thread, defaultValue=<rgba 0 1 1 0>}.RGB
    Alpha=Multiply(A=@n15, B=@n16.Fresnel)
   )
  )
  n18 = VectorParameter() {parameterName=Tire Tread Streaks Color, defaultValue=<rgba 1 0.9117 0.7 0>}
  n19 = TextureCoordinate()
  n20 = Divide(
   A=Subtract(A=ComponentMask(None=@n19) {g=true}, B=ScalarParameter() {defaultValue=-0.1, parameterName=Tire Tread Mask Position})
   B=ScalarParameter() {defaultValue=0.75, parameterName=Tire Tread Mask Softness}
  )
  n21 = TextureCoordinate() {coordinateIndex=1}
  n22 = ScalarParameter() {defaultValue=2.618, parameterName=Tire Tread Overlay Panner Speed A X}
  n23 = ScalarParameter() {defaultValue=0.618, parameterName=Tire Tread Overlay Panner Speed A Y}
  n24 = TextureSample(UVs=Panner(Coordinate=@n21, Time=@n6, Speed=AppendVector(A=@n22, B=@n23)) {bFractionalPart=true}) {
   samplerType=SAMPLERTYPE_Masks
   texture=&T_Contrail_Motorcyclist_Female_Wheel
  }
  n25 = ScalarParameter() {defaultValue=4.23, parameterName=Tire Tread Overlay Panner Speed B X}
  n26 = ScalarParameter() {defaultValue=-0.45, parameterName=Tire Tread Overlay Panner Speed B Y}
  n27 = TextureSample(UVs=Panner(Coordinate=@n21, Time=@n6, Speed=AppendVector(A=@n25, B=@n26)) {bFractionalPart=true}) {
   samplerType=SAMPLERTYPE_Masks
   texture=&T_Contrail_Motorcyclist_Female_Wheel
  }
  n28 = Multiply(A=@n24.G, B=@n27.G)
  n29 = Multiply(
   A=OneMinus(None=Saturate(None=@n20))
   B=Power(
    Base=NamedRerouteDeclaration(None=@n28) {name=Directional Noise, nodeColor=<rgba 0.4297 1 0 1>}
    Exp=ScalarParameter() {defaultValue=2, parameterName=Tire Tread Mask Exponent}
   )
  )
  n30 = TwoSidedSign()
  n31 = Divide(A=ComponentMask(None=TextureCoordinate()) {g=true}, B=ScalarParameter() {defaultValue=0.3, parameterName=Contact Surface Mask Size})
  n32 = OneMinus(None=Power(Base=Saturate(None=@n31), Exp=ScalarParameter() {defaultValue=3, parameterName=Contact Surface Width Exponent}))
  n33 = TextureCoordinate()
  n34 = Frac(None=Subtract(A=ComponentMask(None=@n33) {r=true}, B=ScalarParameter() {defaultValue=0.2, parameterName=Contact Surface Shift, sliderMax=1}))
  n35 = Saturate(None=Divide(A=OneMinus(None=@n34), B=ScalarParameter() {defaultValue=0.6, parameterName=Contact Surface Length}))
  n36 = Power(
   Base=MaterialFunctionCall!(Gradient=Frac(None=@n35), Center Point=ScalarParameter() {defaultValue=0.333, parameterName=Contact Surface Centerpoint, sliderMax=1}) {materialFunction=/Game/Effects/Fort_Effects/MaterialFunctions/MF_GradientReflected}.Result
   Exp=ScalarParameter() {defaultValue=2, parameterName=Contact Surface Length Exponent}
  )
  n37 = MaterialFunctionCall!(
   Gradient=Power(Base=NamedRerouteUsage(), Exp=ScalarParameter() {defaultValue=0.333, parameterName=Contact Surface Mask Exponent})
   Erosion Amount=OneMinus(None=Multiply(A=@n32, B=@n36))
  ) {materialFunction=/BRCosmetics/Athena/Effects/Particles/Skydiving/NS_Contrail_DarkAndLight/Functions/MF_ErosionNormalized}
  n38 = ScalarParameter() {defaultValue=3, parameterName=Contact Surface Final Exponent}
  n39 = LinearInterpolate(
   A=LinearInterpolate(
    A=VectorParameter() {parameterName=Color Inside, defaultValue=<rgba 0 0.15 0.3 0>}.RGB
    B=LinearInterpolate(A=@n17, B=@n18.RGB, Alpha=@n29)
    Alpha=Saturate(None=@n30)
   )
   B=VectorParameter() {parameterName=Contact Surface, defaultValue=<rgba 0 6 6 0>}.RGB
   Alpha=Saturate(None=Power(Base=@n37.Result, Exp=@n38))
  )
  n40 = VectorParameter() {parameterName=Tire Center Color, defaultValue=<rgba 0.0296 0.4 0.295 1>}
  n41 = TextureCoordinate()
  n42 = Divide(
   A=Subtract(A=ComponentMask(None=@n41) {g=true}, B=ScalarParameter() {defaultValue=0.75, parameterName=Tire Center Glow Offset})
   B=ScalarParameter() {defaultValue=0.1, parameterName=Tire Center Glow Softness}
  )
  n43 = Power(Base=Multiply(A=Saturate(None=@n42), B=@n10), Exp=@n40.A) {constExponent=3}
  n44 = VectorParameter() {parameterName=Comet Color, defaultValue=<rgba 0.1 0.2637 1 1>}
  n45 = Multiply(A=@n44.A, B=NamedRerouteUsage())
  n46 = Add(
   A=Multiply(A=Power(Base=@n4, Exp=@n5.A), B=@n5.RGB)
   B=Add(A=LinearInterpolate(A=@n39, B=@n40.RGB, Alpha=@n43), B=Multiply(A=@n44.RGB, B=@n45))
  )
  MaterialFunctionCall(Emissive=Multiply(
    A=ParticleColor().A
    B=MaterialFunctionCall(Hue Shift Percentage=@n3, Texture=@n46) {materialFunction=&HueShift}.Result
   )) {materialFunction=&MF_FX_EmissiveAdjustments}
  n47 = ScalarParameter() {defaultValue=7, parameterName=Comet UV Scale U}
  n48 = ScalarParameter() {defaultValue=0.2, parameterName=Comet UV Scale V}
  n49 = Add(A=Multiply(A=TextureCoordinate() {coordinateIndex=2}, B=AppendVector(A=@n47, B=@n48)))
  n50 = ScalarParameter() {defaultValue=-0.17, parameterName=Comet Overlay Panner Speed A X}
  n51 = ScalarParameter() {defaultValue=1.618, parameterName=Comet Overlay Panner Speed A Y}
  n52 = TextureSample(UVs=Panner(Coordinate=@n49, Speed=AppendVector(A=@n50, B=@n51)) {bFractionalPart=true}) {
   samplerType=SAMPLERTYPE_Masks
   texture=&T_Contrail_Motorcyclist_Female_Wheel
  }
  n53 = ScalarParameter() {defaultValue=0.2, parameterName=Comet Overlay Panner Speed B X}
  n54 = ScalarParameter() {defaultValue=0.618, parameterName=Comet Overlay Panner Speed B Y}
  n55 = TextureSample(UVs=Panner(Coordinate=@n49, Speed=AppendVector(A=@n53, B=@n54)) {bFractionalPart=true}) {
   samplerType=SAMPLERTYPE_Masks
   texture=&T_Contrail_Motorcyclist_Female_Wheel
  }
  n56 = ComponentMask(None=MaterialFunctionCall!(Blend=@n52.B, Base=@n55.B) {materialFunction=/Engine/Functions/Engine_MaterialFunctions03/Blends/Blend_Screen}.Result) {r=true}
  n57 = TextureCoordinate() {coordinateIndex=3}
  n58 = Abs(None=Sine(None=ComponentMask(None=@n57) {r=true}))
  n59 = ScalarParameter() {defaultValue=4, parameterName=Comet Detail Erosion Blend Exponent}
  n60 = Power(
   Base=OneMinus(None=ComponentMask(None=@n57) {g=true})
   Exp=LinearInterpolate(
    A=ScalarParameter() {defaultValue=0.666, parameterName=Comet Detail Erosion Exponent Tread}
    B=ScalarParameter() {defaultValue=1.6, parameterName=Comet Detail Erosion Exponent Face}
    Alpha=Power(Base=@n58, Exp=@n59)
   )
  )
  n61 = ScalarParameter() {defaultValue=3, parameterName=Comet Erosion Softness}
  NamedRerouteDeclaration(None=Power(
    Base=MaterialFunctionCall!(Gradient=@n56, Erosion Amount=@n60, Softness=@n61) {materialFunction=/BRCosmetics/Athena/Effects/Particles/Skydiving/NS_Contrail_DarkAndLight/Functions/MF_Erosion}.Increased Range
    Exp=ScalarParameter() {defaultValue=6, parameterName=Comet Erosion Exponent}
   )) {name=Comet Mask, nodeColor=<rgba 1 0 0.75 1>}
 ]
}
