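# Skydiving contrail Niagara system for the female Motorcyclist cosmetic, attaching glowing tires, chains, road trails and sparks to two skeletal bones, apparently the hands.
# Road_Ribbons computes separate position, width and color attributes for contact, smoke and flame layers across five ribbon renderers.
# Invisible Spark_Ribbon_Guides feed Spark_Ribbons through particle reads, alongside spark sprites.
# Owner velocity drives gameplay motion, front-end overrides use glide lean, and time of day scales brightness.
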NiagaraSystem /BRCosmetics/Athena/Effects/Particles/Skydiving/NS_Contrail_Motorcyclist_Female/NS_Contrail_Motorcyclist_Female
dependencies {
 usedDataInterfaces [
  &NiagaraDataInterfaceCurve
  &NiagaraDataInterfaceColorCurve
  &NiagaraDataInterfaceArrayInt32
  &NiagaraDataInterfaceArrayFloat3
  &NiagaraDataInterfaceParticleRead
 ]
 usedDynamicInputs [/Niagara/DynamicInputs/Vectors/CrossProduct, &MakeFloatFromVector]
 usedModules [
  &self:SystemSpawnScript.SystemScriptSource.SystemScriptGraph.NiagaraNodeAssignment_2.SetVariables_#1
  &SystemState
  &self:SystemSpawnScript.SystemScriptSource.SystemScriptGraph.NiagaraNodeAssignment_0.SetVariables_#2
  &self:SystemSpawnScript.SystemScriptSource.SystemScriptGraph.NiagaraNodeAssignment_1.SetVariables_#3
  /Game/Effects/Niagara/Modules/Skydiving/FN_DivingAttributes/v4/FN_DivingAttributes_v4
 ]
}
emitters {
 Chains {
  emitterSpawnScript [&FN_PreserveSkydiveUserVariables() {module=PreserveUserVariables}]
  emitterUpdateScript [
   &EmitterState(Inactive Response=::NewEnumerator1, Life Cycle Mode=::NewEnumerator1, Loop Duration=0, UseLoopDelay=false)
   &SpawnBurst_Instantaneous(Spawn Count=8, Spawn Probability=HlslExpression() {hlslExpression="Emitter.LoopCount == 0"}, Use Spawn Probability=true) {module=SpawnBurst_Instantaneous001}
  ]
  particleSpawnScript [
   &FN_SetSpawnGroup(Number of Spawn Groups=2)
   &InitializeParticle(
    Color Mode=::NewEnumerator0
    Lifetime=0
    Sprite Size Mode=::NewEnumerator3
    Uniform Sprite Size=40
    UsePositionOffset=false
    Write Material Random=true
   ) {module=InitializeParticle001}
   &Color(Color=$NPC.FortniteNPC.FortniteActiveTimeOfDay)
   &DynamicMaterialParameters(
    Index 0 Param 1=&RandomRangeFloat(Minimum=0, Maximum=3.333)
    Index 0 Param 2=&FN_FrontEndFloatModifier(Gameplay=5, Front End=4)
    Index 0 Param 3=&FN_FrontEndFloatModifier(Gameplay=0, Front End=15)
    Index 0 Param 4=&FN_FrontEndFloatModifier(Gameplay=55, Front End=22)
    Module.Param0WriteEnabled=true
    Module.Param1WriteEnabled=true
    Module.Param2WriteEnabled=true
    Module.Param3WriteEnabled=true
   )
   /SetVariables(
    Particles.MeshIndex=&SelectIntFromArray(
     Direct Array Index=&ReturnParticleIDAttributes()
     Direct Array Mode=::NewEnumerator1
     Int Selection Array=DataInterface() {propertyValues {IntData [0,1,2,3,1,2,3,0]}}
    )
    Particles.Scale=&FN_FrontEndVector3Modifier(Gameplay=<1 1 1>, Front End=<0.8 0.8 0.8>)
   ) {isSetParameters=true}
   &InitialMeshOrientation(
    Mesh Orientation Mode=::NewEnumerator7
    Rotation=&MakeVector(X=&RandomRangeFloat(Minimum=0, Maximum=1), Y=0, Z=0)
    Rotation Coordinate Space=::NewEnumerator0
    Use Rotation Vector=true
   )
  ]
  particleUpdateScript [
   &ParticleState(Kill Particles When Lifetime Has Elapsed=false)
   &SkeletalMeshLocation(
    Bone / Socket Index=$Particles.SpawnGroup
    Bone Sampling Mode=::NewEnumerator4
    Inherit Velocity=false
    Interpolate Mesh Sampling=false
    MaskSpawnGroup=false
    Sampled Position Offset=&MaskVectorBySpawnGroup(Masked Vector=<9 -2 1>, Passthrough Vector=<-9 2 -1>)
    Sampled Position Offset Coordinate Space:ENiagaraMeshTransforms=::NewEnumerator3
    Skeletal Mesh=$User.SkeletalMesh
   )
   /Niagara/Modules/Update/Orientation/V2/OrientMeshToVector(
    Look At Direction:Vector3f=&NormalizeVector(
     Vector To Normalize=&Add_Vector(
      A=&NormalizeVector(
       Vector To Normalize=&FN_FrontEndVector3Modifier(
        Gameplay=&Multiply_VectorByFloat(Vector=$Engine.Owner.Velocity, Float=-1)
        Front End=&TransformVector(
         Vector=&NormalizeVector(
          Vector To Normalize=&Lerp_Vector(A=<0 -0.2 1>, B=<0 -2 1>, Alpha=$User.ParaGlideLeanAlpha)
          Safe Normalize=true
          Safe Fallback Threshold=0
          Safe Fallback Vector=<0 0 0.001>
         )
         Destination Space=::Simulation
        )
       )
       Safe Normalize=false
       Safe Fallback Threshold=0
       Safe Fallback Vector=<0 0 0.001>
      )
      B=&Multiply_VectorByFloat(
       Vector=&SelectVectorFromArray(
        Direct Array Index=&ReturnParticleIDAttributes()
        Direct Array Mode=::NewEnumerator1
        Vector Selection Array=DataInterface() {propertyValues {FloatData [<1 0 0>, <1 0 0>, <0 1 0>, <0 1 0>, <-1 0 0>, <-1 0 0>, <0 -1 0>, <0 -1 0>]}}
       )
       Float=&RandomRangeFloat(Minimum=0.166, Maximum=0.266)
      )
     )
     Safe Normalize=false
     Safe Fallback Threshold=0
     Safe Fallback Vector=<0 0 0.001>
    )
    Override Quat=false
    Side Axis Direction:Vector3f=&TransformVector(Vector=<0 1 0>, Destination Space=::World)
    UseOptionalSideAxis=false
    UseOptionalUpAxis=false
   )
   &UpdateMeshOrientation(
    Rotation Coordinate Space=::NewEnumerator3
    Rotation Rate=&RandomRangeFloat(Minimum=-1, Maximum=1)
    Rotation Vector=<1 0 0>
   )
  ]
  rendererClasses [&NiagaraMeshRendererProperties]
  renderers [{rendererClass=&NiagaraMeshRendererProperties, rendererIndex=0}]
 }
 FN_Debug {
  bEnabled=false
  emitterUpdateScript [
   &EmitterState()
   &SpawnBurst_Instantaneous(Spawn Count=1, Use Spawn Probability=false)
   &SpawnRate(SpawnRate=3, Use Spawn Probability=false)
  ]
  particleSpawnScript [/Niagara/Modules/Spawn/Initialization/InitializeParticle(
    Lifetime=5
    Module.Write Color=true
    Module.Write Lifetime=true
    Module.Write Mass=true
    Module.Write Material Random=true
    Module.Write Position=false
    Module.Write Scale=false
    Module.Write SpriteRotation=false
    Module.Write SpriteSize=true
    Position=&SimulationPosition()
    Sprite Size=<100 100>
   )]
  particleUpdateScript [
   &ParticleState()
   &FN_Timer()
   /SetVariables(Particles.Debug:NiagaraFloat=&FNTimeOfDayScalar(Night Scale=0.4, Morning Scale=0.5, Evening Scale=0.4)) {isSetParameters=true}
   &DynamicMaterialParameters(
    Index 0 Param 1=$Particles.Debug
    Module.Param0WriteEnabled=true
    Module.Param1WriteEnabled=true
    Module.Param2WriteEnabled=true
    Module.Param3WriteEnabled=true
   )
   &CameraOffset(Camera Offset Amount=111)
  ]
  rendererClasses [&NiagaraSpriteRendererProperties]
  renderers [{rendererClass=&NiagaraSpriteRendererProperties}]
 }
 Road_Ribbons {
  emitterUpdateScript [
   &EmitterState(Life Cycle Mode=::NewEnumerator1, Loop Duration=0, UseLoopDelay=false)
   /Game/Effects/Niagara/Modules/Skydiving/FN_DivingAttributes/FN_EmitterDivingAttributes_v1(Bone Index=1, SkeletalMesh:NiagaraDataInterfaceSkeletalMesh=$User.SkeletalMesh)
   &FN_SpawnRibbon_v2(Number of Ribbons=2, Spawn Mode=::NewEnumerator0, SpawnRate=128)
  ]
  particleSpawnScript [
   &FN_SetSpawnGroup(Number of Spawn Groups=2)
   &DynamicMaterialParameters(
    Index 0 Param 1=&MaskFloatBySpawnGroup(Masked Float=$System.ZPositionLeftHand, Passthrough Float=$System.ZPositionRightHand)
    Index 0 Param 2=&MaskFloatBySpawnGroup(Masked Float=0, Passthrough Float=0.65)
    Module.Param0WriteEnabled=true
    Module.Param1WriteEnabled=false
    Module.Param2WriteEnabled=false
    Module.Param3WriteEnabled=false
   ) {module=DynamicMaterialParameters001}
   &FN_InitializeRibbon_v3(
    Color:LinearColor=&MakeLinearColorFromVectorAndFloat(
     "Vector (RGB)"=<0.0093 0 0.0125>
     "Float (Alpha)"=&FNTimeOfDayScalar(Night Scale=0.4, Morning Scale=0.5, Evening Scale=0.4)
    )
    Custom Offset=true
    Enable Skeletal Position=true
    Facing Axis:ENiagaraOrientationAxis=::YAxis
    Lifetime=0.2
    Local Vector Offset:Vector3f=<0 1 0>
    Ribbon Width=19
   )
   &SkeletalMeshLocation(
    Bone / Socket Index=$Particles.SpawnGroup
    Bone Sampling Mode=::NewEnumerator4
    Inherit Velocity=false
    Interpolate Mesh Sampling=false
    MaskSpawnGroup=false
    Sampled Position Offset=&MaskVectorBySpawnGroup(Masked Vector=<9 -2 -1>, Passthrough Vector=<-9 2 1>)
    Sampled Position Offset Coordinate Space:ENiagaraMeshTransforms=::NewEnumerator3
    Skeletal Mesh=$User.SkeletalMesh
   )
   /SetVariables(
    Particles.CometFlamesPosition:Vector3f=&Add_Vector(
     A=Linked() {linkedVariable {name=Particles.Position, type=NiagaraPosition}}
     B=&Multiply_VectorByFloat(Vector=$System.VelocityOffset, Float=2)
    )
    Particles.Position=&Add_Vector(A=$System.TireOffset, B=Linked() {linkedVariable {name=Particles.Position, type=NiagaraPosition}})
    Particles.RoadRibbonFacing:Vector3f=$Engine.Owner.SystemXAxis
    Particles.RoadRibbonWidth:NiagaraFloat=$Particles.RibbonWidth
    Particles.SmokePosition:Vector3f=&Add_Vector(
     A=&Add_Vector(A=$System.TireOffset, B=$System.VelocityOffset)
     B=Linked() {linkedVariable {name=Particles.Position, type=NiagaraPosition}}
    )
   ) {isSetParameters=true}
   /SetVariables(
    Particles.ContactTrailFlatWidth:NiagaraFloat=$Particles.RoadRibbonWidth
    Particles.ContactTrailPosition:Vector3f=&Add_Vector(
     A=Linked() {linkedVariable {name=Particles.Position, type=NiagaraPosition}}
     B=&TransformVector(
      Vector=&MaskVectorBySpawnGroup(Masked Vector=<4 0 0>, Passthrough Vector=<-4 0 0>)
      Destination Space=::Simulation
     )
    )
   ) {isSetParameters=true}
   &AddVelocity(Velocity=&FN_FrontEndVector3Modifier(
     Gameplay=&Multiply_Vector(A=<22 66 0>, B=&MaskVectorBySpawnGroup(Masked Vector=<1 -1 1>, Passthrough Vector=<-1 -1 1>))
     Front End=&Multiply_VectorByFloat(
      Vector=&NormalizeVector(
       Vector To Normalize=&Lerp_Vector(A=<0 0.2 -1>, B=<0 2 -1>, Alpha=$User.ParaGlideLeanAlpha)
       Safe Normalize=true
       Safe Fallback Threshold=0
       Safe Fallback Vector=<0 0 0.001>
      )
      Float=&Lerp_Float(A=-666, B=-1111, Alpha=$User.ParaGlideLeanAlpha)
     )
    ))
  ]
  particleUpdateScript [
   /SetVariables(
    Particles.CometFlamesColor:LinearColor=&MakeLinearColorFromVectorAndFloat(
     "Vector (RGB)"=<1 1 1>
     "Float (Alpha)"=&FloatFromCurve(
      FloatCurve=DataInterface() {propertyValues {Curve {keys [<at 0 0 RCIM_Constant>,<at 0.0166 1>,<at 0.4 1>,<at 1 0>]}}}
      Scale Curve=&FNTimeOfDayScalar(Night Scale=1, Morning Scale=2, Evening Scale=0.6)
     )
    )
    Particles.CometFlamesPosition:Vector3f=&FN_FrontEndVector3Modifier(
     Gameplay=$Particles.CometFlamesPosition
     Front End=&Add_Vector(
      A=Linked() {linkedVariable {name=Particles.Position, type=NiagaraPosition}}
      B=&Multiply_VectorByFloat(
       Vector=&NormalizeVector(
        Vector To Normalize=&TransformVector(Vector=&Lerp_Vector(A=<0 0.7 -0.7>, B=<0 1 0>, Alpha=$User.ParaGlideLeanAlpha), Destination Space=::Simulation)
        Safe Normalize=true
        Safe Fallback Threshold=0
        Safe Fallback Vector=<0 0 0.001>
       )
       Float=33
      )
     )
    )
    Particles.CometFlamesWidth:NiagaraFloat=&FloatFromCurve(
     FloatCurve=DataInterface() {propertyValues {Curve {keys [
         {interpMode=RCIM_Cubic, leaveTangent=4.9154, tangentMode=RCTM_Break, time=0, value=0.5}
         {arriveTangentWeight=0.074, interpMode=RCIM_Cubic, leaveTangentWeight=0.2219, tangentMode=RCTM_User, time=0.2, value=1}
         {arriveTangent=-3.1066, interpMode=RCIM_Cubic, leaveTangent=-3.1066, tangentMode=RCTM_User, time=0.8, value=0}
        ]}}}
     Scale Curve=44
    )
    Particles.ContactTrailColor:LinearColor=&ColorFromCurve(ColorCurve=DataInterface() {propertyValues {
       AlphaCurve {keys [<at 0 1 RCIM_Cubic>,<at 0.13 0 RCIM_Cubic>]}
       BlueCurve {keys [
         {arriveTangent=-24.1394, interpMode=RCIM_Cubic, leaveTangent=-24.1394, time=0, value=4}
         {arriveTangent=-24.1394, interpMode=RCIM_Cubic, leaveTangent=-24.1394, time=0.1243, value=1}
        ]}
       GreenCurve {keys [
         {arriveTangent=-24.1394, interpMode=RCIM_Cubic, leaveTangent=-24.1394, time=0, value=4}
         {arriveTangent=-24.1394, interpMode=RCIM_Cubic, leaveTangent=-24.1394, time=0.1243, value=1}
        ]}
       RedCurve {keys [<at 0 0 RCIM_Cubic>,<at 0.1243 0 RCIM_Cubic>]}
      }})
    Particles.ContactTrailFlatWidth:NiagaraFloat=&Add_Float(
     A=$Particles.RoadRibbonWidth
     B=&FloatFromCurve(
      FloatCurve=DataInterface() {propertyValues {Curve {keys [{tangentMode=RCTM_User, time=0, value=0.5}, <at 0.2 1 RCIM_Constant>, <at 0.3 0>]}}}
      Scale Curve=11
     )
    )
    Particles.ContactTrailSideWidth:NiagaraFloat=&FloatFromCurve(
     FloatCurve=DataInterface() {propertyValues {Curve {keys [{arriveTangent=10.885, time=0, value=0.4}, <at 0.15 1 RCIM_Constant>, <at 0.25 0 RCIM_Constant>]}}}
     Scale Curve=33
    )
    Particles.SmokePosition:Vector3f=&FN_FrontEndVector3Modifier(Gameplay=$Particles.SmokePosition, Front End=Linked() {linkedVariable {name=Particles.Position, type=NiagaraPosition}})
   ) {isSetParameters=true}
   /SetVariables(Particles.ContactTrailPosition:Vector3f=&Add_Vector(
     A=&Add_Vector(
      A=Linked() {linkedVariable {name=Particles.Position, type=NiagaraPosition}}
      B=&Multiply_VectorByFloat(
       Vector=&TransformVector(Vector=<0 0.5 0>, Destination Space=::Simulation)
       Float=&Divide_Float(A=$Particles.ContactTrailSideWidth, B=2)
      )
     )
     B=&TransformVector(
      Vector=&MaskVectorBySpawnGroup(Masked Vector=<4 0 0>, Passthrough Vector=<-4 0 0>)
      Destination Space=::Simulation
     )
    )) {isSetParameters=true}
   &SampleSkeletalMesh(
    AdditionalYawPitchRoll=false
    Bone / Socket Index:NiagaraInt32=&SpawnGroup()
    Bone Sampling Mode:ENiagaraBoneSamplingMode=::NewEnumerator4
    Sampled Position Offset:Vector3f=&MaskVectorBySpawnGroup(Masked Vector=<9 -2 -1>, Passthrough Vector=<-9 2 1>)
    Skeletal Mesh:NiagaraDataInterfaceSkeletalMesh=$User.SkeletalMesh
   )
   &DynamicMaterialParameters(
    Index 0 Param 1=&MaskFloatBySpawnGroup(Masked Float=$System.ZPositionLeftHand, Passthrough Float=$System.ZPositionRightHand)
    Index 0 Param 2=&MaskFloatBySpawnGroup(Masked Float=0, Passthrough Float=0.65)
    Module.Param0WriteEnabled=true
    Module.Param1WriteEnabled=true
    Module.Param2WriteEnabled=false
    Module.Param3WriteEnabled=false
   )
   &SolveForcesAndVelocity(Clamp Velocity=false, Limit Acceleration=false)
   &FN_UpdateRibbon(Detach Locked Particle=true)
   &ScaleRibbonWidth(Ribbon Width Scale=&FloatFromCurve(FloatCurve=DataInterface() {propertyValues {Curve {keys [{tangentMode=RCTM_User, time=0, value=1}, {tangentMode=RCTM_User, time=1, value=9}]}}}))
  ]
  rendererClasses [&NiagaraRibbonRendererProperties]
  renderers [
   {rendererClass=&NiagaraRibbonRendererProperties}
   {rendererClass=&NiagaraRibbonRendererProperties, rendererIndex=1}
   {rendererClass=&NiagaraRibbonRendererProperties, rendererIndex=2}
   {rendererClass=&NiagaraRibbonRendererProperties, rendererIndex=3}
   {rendererClass=&NiagaraRibbonRendererProperties, rendererIndex=4}
  ]
 }
 Spark_Ribbon_Guides {
  emitterUpdateScript [
   &EmitterState()
   &SpawnRate(
    Spawn Probability=0.3
    SpawnRate=&FN_FrontEndFloatModifier(Gameplay=11, Front End=5)
    Use Spawn Probability=true
   )
  ]
  particleSpawnScript [
   /Niagara/Modules/Spawn/Initialization/V2/InitializeParticle(Color=<rgba 0 17.05 33 1>, Lifetime=9, Sprite Size Mode=::NewEnumerator3, UsePositionOffset=false, Write Material Random=true)
   &FN_SetSpawnGroup(Number of Spawn Groups=2)
   &SkeletalMeshLocation(
    Bone / Socket Index=$Particles.SpawnGroup
    Bone Sampling Mode=::NewEnumerator4
    Inherit Velocity=false
    Interpolate Mesh Sampling=false
    MaskSpawnGroup=false
    Sampled Position Offset=&MaskVectorBySpawnGroup(Masked Vector=<9 -2 -1>, Passthrough Vector=<-9 2 1>)
    Sampled Position Offset Coordinate Space:ENiagaraMeshTransforms=::NewEnumerator3
    Skeletal Mesh=$User.SkeletalMesh
   )
   /SetVariables(
    Particles.Lifetime=&Lerp_Float(A=0.066, B=0.233, Alpha=$Particles.MaterialRandom)
    Particles.Position=&Add_Vector(A=$System.TireOffset, B=Linked() {linkedVariable {name=Particles.Position, type=NiagaraPosition}})
   ) {isSetParameters=true}
   &AddVelocity(Coordinate Space=::World, Velocity=&Multiply_VectorByFloat(Vector=$Engine.Owner.Velocity, Float=-1)) {module=AddVelocity001}
   &AddVelocity(
    Scale Added Velocity=&Lerp_Vector(A=<1 1 1>, B=<4 4 1>, Alpha=$System.Dive)
    Velocity=&FN_FrontEndVector3Modifier(
     Gameplay=&Multiply_VectorByFloat(
      Vector=&Multiply_Vector(
       A=&MakeVector(
        X=&RandomRangeFloat(Minimum=99, Maximum=-99)
        Y=&Lerp_Float(A=66, B=-99, Alpha=$Particles.MaterialRandom)
        Z=0
       )
       B=&MaskVectorBySpawnGroup(Masked Vector=<1 -1 1>, Passthrough Vector=<-1 -1 1>)
      )
      Float=1
     )
     Front End=&Multiply_VectorByFloat(
      Vector=&NormalizeVector(
       Vector To Normalize=&Lerp_Vector(A=<0 0 -1>, B=<0 0.6 -1>, Alpha=$User.ParaGlideLeanAlpha)
       Safe Normalize=true
       Safe Fallback Threshold=0
       Safe Fallback Vector=<0 0 0.001>
      )
      Float=&Lerp_Float(A=-444, B=-666, Alpha=$User.ParaGlideLeanAlpha)
     )
    )
   )
   &BoxLocation(Box Size=<10 0 0>, MaskSpawnGroup=false, Module.Override Seed=false, Rotate Box=false, Surface Only=false)
  ]
  particleUpdateScript [
   &ParticleState()
   &InheritSourceMovement(ApplyForce=false, ApplyPosition=true, ApplyPositionFalloff=false)
   &AccelerationForce(
    Acceleration=&FN_FrontEndVector3Modifier(
     Gameplay=&Multiply_VectorByFloat(
      Vector=&MakeVector(
       X=0
       Y=&Multiply_Float(A=-3333, B=$Particles.MaterialRandom)
       Z=&RandomRangeFloat(Minimum=0, Maximum=0)
      )
      Float=$Particles.MaterialRandom
     )
     Front End=&MakeVector(
      X=0
      Y=&Multiply_Float(A=0, B=$Particles.MaterialRandom)
      Z=&Multiply_Float(A=0, B=$Particles.MaterialRandom)
     )
    )
    Coordinate Space=::Local
   )
   &Drag(Drag=&RandomRangeFloat(Minimum=2, Maximum=22), Use Linear Drag=true, Use Rotational Drag=false)
   &SolveForcesAndVelocity(Clamp Velocity=false, Limit Acceleration=false)
   &ScaleColor(
    Scale RGB=&Multiply_VectorByFloat(
     Vector=<1 1 1>
     Float=&FloatFromCurve(FloatCurve=DataInterface() {propertyValues {Curve {keys [
          {arriveTangent=9.0725, interpMode=RCIM_Cubic, leaveTangent=9.0725, tangentMode=RCTM_User, time=0, value=0}
          {arriveTangent=-2.5249, interpMode=RCIM_Cubic, leaveTangent=-2.5249, tangentMode=RCTM_User, time=0.1408, value=0.9847}
          {arriveTangent=-0.0405, interpMode=RCIM_Cubic, leaveTangent=-0.0405, tangentMode=RCTM_User, time=1, value=0}
         ]}}})
    )
    ScaleA=true
    ScaleRGB=true
   )
  ]
  rendererClasses []
  renderers []
 }
 Spark_Ribbons {
  emitterUpdateScript [
   &EmitterState()
   &SpawnParticlesFromOtherEmitter(
    Attribute Reader=DataInterface() {propertyValues {EmitterBinding {bindingMode=Other, emitterName=Spark_Ribbon_Guides}}}
    Spawn Rate=333
    Use Maximum Age=false
    Use Minimum Age=false
   )
  ]
  particleSpawnScript [
   &InitializeParticle(UsePositionOffset=false, Write Material Random=true)
   &InitializeRibbon(Ribbon Width=0.5, UsePositionOffset=false, Write Material Random=true)
   &SampleParticlesFromOtherEmitter(
    Color Sampling=::NewEnumerator0
    Lifetime Sampling:ENiagara_DirectReadSamplingMode=::NewEnumerator0
    Particle ID Sampling=::NewEnumerator1
    Reject Particle on Sampled Particle's First Frame=true
   )
   &AddVelocity(Coordinate Space=::World, Velocity=$Engine.Owner.Velocity)
  ]
  particleUpdateScript [
   &SolveForcesAndVelocity(Clamp Velocity=false, Limit Acceleration=false)
   &ParticleState()
   &ScaleRibbonWidth(Ribbon Width Scale=&FloatFromCurve(FloatCurve=DataInterface() {propertyValues {Curve {keys [
         {arriveTangent=9.0725, interpMode=RCIM_Cubic, leaveTangent=9.0725, tangentMode=RCTM_User, time=0, value=0}
         {arriveTangent=-2.5249, interpMode=RCIM_Cubic, leaveTangent=-2.5249, tangentMode=RCTM_User, time=0.1408, value=0.9847}
         {arriveTangent=-0.0405, interpMode=RCIM_Cubic, leaveTangent=-0.0405, tangentMode=RCTM_User, time=1, value=0}
        ]}}}))
   &ScaleColor(
    Scale Alpha=&FNTimeOfDayScalar(Night Scale=0.4, Morning Scale=0.5, Evening Scale=0.4)
    Scale RGB=&Multiply_VectorByFloat(
     Vector=<1 1 1>
     Float=&FloatFromCurve(
      FloatCurve=DataInterface() {propertyValues {Curve {keys [<at 0 1>,<at 1 0>]}}}
      Scale Curve=&FNTimeOfDayScalar(Night Scale=0.4, Morning Scale=0.5, Evening Scale=0.4)
     )
    )
    ScaleA=true
    ScaleRGB=true
   )
  ]
  rendererClasses [&NiagaraRibbonRendererProperties]
  renderers [{rendererClass=&NiagaraRibbonRendererProperties}]
 }
 Spark_Sprites {
  emitterUpdateScript [
   &EmitterState()
   &SpawnRate(
    SpawnRate=&FN_FrontEndFloatModifier(
     Gameplay=&RandomRangeFloat(Minimum=5, Maximum=33)
     Front End=&RandomRangeFloat(Minimum=3, Maximum=9)
    )
    Use Spawn Probability=false
   )
  ]
  particleSpawnScript [
   &InitializeParticle(
    Color=<rgba 0 51.1499 99 1>
    Sprite Size Max=<1.333 8>
    Sprite Size Min=<0.2 3>
    Sprite Size Mode=::NewEnumerator2
    UsePositionOffset=false
    Write Material Random=true
   )
   &FN_SetSpawnGroup(Number of Spawn Groups=2)
   &SkeletalMeshLocation(
    Bone / Socket Index=$Particles.SpawnGroup
    Bone Sampling Mode=::NewEnumerator4
    Inherit Velocity=false
    Interpolate Mesh Sampling=false
    MaskSpawnGroup=false
    Sampled Position Offset=&MaskVectorBySpawnGroup(Masked Vector=<9 -2 -1>, Passthrough Vector=<-9 2 1>)
    Sampled Position Offset Coordinate Space:ENiagaraMeshTransforms=::NewEnumerator3
    Skeletal Mesh=$User.SkeletalMesh
   )
   /SetVariables(
    Particles.Lifetime=&Lerp_Float(A=0.066, B=0.233, Alpha=$Particles.MaterialRandom)
    Particles.Position=&Add_Vector(A=$System.TireOffset, B=Linked() {linkedVariable {name=Particles.Position, type=NiagaraPosition}})
   ) {isSetParameters=true}
   &AddVelocity(Coordinate Space=::World, Velocity=&Multiply_VectorByFloat(Vector=$Engine.Owner.Velocity, Float=-1)) {module=AddVelocity001}
   &AddVelocity(
    Scale Added Velocity=&Lerp_Vector(A=<1 1 1>, B=<4 4 1>, Alpha=$System.Dive)
    Velocity=&FN_FrontEndVector3Modifier(
     Gameplay=&Multiply_VectorByFloat(
      Vector=&Multiply_Vector(
       A=&MakeVector(
        X=&RandomRangeFloat(Minimum=99, Maximum=-99)
        Y=&Lerp_Float(A=66, B=-99, Alpha=$Particles.MaterialRandom)
        Z=0
       )
       B=&MaskVectorBySpawnGroup(Masked Vector=<1 -1 1>, Passthrough Vector=<-1 -1 1>)
      )
      Float=1
     )
     Front End=&Multiply_VectorByFloat(
      Vector=&NormalizeVector(
       Vector To Normalize=&Lerp_Vector(A=<0 0.05 -1>, B=<0 0.6 -1>, Alpha=$User.ParaGlideLeanAlpha)
       Safe Normalize=true
       Safe Fallback Threshold=0
       Safe Fallback Vector=<0 0 0.001>
      )
      Float=&Lerp_Float(A=-444, B=-666, Alpha=$User.ParaGlideLeanAlpha)
     )
    )
   )
   &BoxLocation(Box Size=<10 0 0>, MaskSpawnGroup=false, Module.Override Seed=false, Rotate Box=false, Surface Only=false)
  ]
  particleUpdateScript [
   &ParticleState()
   &InheritSourceMovement(ApplyForce=false, ApplyPosition=true, ApplyPositionFalloff=false)
   &AccelerationForce(
    Acceleration=&FN_FrontEndVector3Modifier(
     Gameplay=&Multiply_VectorByFloat(
      Vector=&MakeVector(
       X=0
       Y=&Multiply_Float(A=-3333, B=$Particles.MaterialRandom)
       Z=&RandomRangeFloat(Minimum=0, Maximum=0)
      )
      Float=$Particles.MaterialRandom
     )
     Front End=&MakeVector(
      X=0
      Y=&Multiply_Float(A=0, B=$Particles.MaterialRandom)
      Z=&Multiply_Float(A=0, B=$Particles.MaterialRandom)
     )
    )
    Coordinate Space=::Local
   )
   &Drag(Drag=&RandomRangeFloat(Minimum=2, Maximum=22), Use Linear Drag=true, Use Rotational Drag=false)
   &SolveForcesAndVelocity(Clamp Velocity=false, Limit Acceleration=false)
   &ScaleColor(
    Scale Alpha=&FNTimeOfDayScalar(Night Scale=0.4, Morning Scale=0.5, Evening Scale=0.4)
    Scale RGB=&Multiply_VectorByFloat(
     Vector=<1 1 1>
     Float=&FloatFromCurve(
      FloatCurve=DataInterface() {propertyValues {Curve {keys [
          {arriveTangent=9.0725, interpMode=RCIM_Cubic, leaveTangent=9.0725, tangentMode=RCTM_User, time=0, value=0}
          {arriveTangent=-2.5249, interpMode=RCIM_Cubic, leaveTangent=-2.5249, tangentMode=RCTM_User, time=0.1408, value=0.9847}
          {arriveTangent=-0.0405, interpMode=RCIM_Cubic, leaveTangent=-0.0405, tangentMode=RCTM_User, time=1, value=0}
         ]}}}
      Scale Curve=&FNTimeOfDayScalar(Night Scale=0.4, Morning Scale=0.5, Evening Scale=0.4)
     )
    )
    ScaleA=true
    ScaleRGB=true
   )
  ]
  rendererClasses [&NiagaraSpriteRendererProperties]
  renderers [{rendererClass=&NiagaraSpriteRendererProperties}]
 }
 Tires {
  emitterSpawnScript [&FN_PreserveSkydiveUserVariables() {module=PreserveUserVariables}]
  emitterUpdateScript [
   &EmitterState(Inactive Response=::NewEnumerator1, Life Cycle Mode=::NewEnumerator1, Loop Duration=0, UseLoopDelay=false)
   &SpawnBurst_Instantaneous(Spawn Count=2, Spawn Probability=HlslExpression() {hlslExpression="Emitter.LoopCount == 0"}, Use Spawn Probability=true) {module=SpawnBurst_Instantaneous001}
  ]
  particleSpawnScript [
   &FN_SetSpawnGroup(Number of Spawn Groups=2)
   &InitializeParticle(
    Color Mode=::NewEnumerator0
    Lifetime=0
    Sprite Size Mode=::NewEnumerator3
    Uniform Sprite Size=15
    UsePositionOffset=false
    Write Material Random=true
   ) {module=InitializeParticle001}
   &DynamicMaterialParameters(
    Index 0 Param 1=&MaskFloatBySpawnGroup(Masked Float=0.5, Passthrough Float=0)
    Index 0 Param 2=&FN_FrontEndFloatModifier(Gameplay=0, Front End=1)
    Module.Param0WriteEnabled=true
    Module.Param1WriteEnabled=true
    Module.Param2WriteEnabled=false
    Module.Param3WriteEnabled=false
   ) {enabled=false}
  ]
  particleUpdateScript [
   &ParticleState(Kill Particles When Lifetime Has Elapsed=false)
   &SkeletalMeshLocation(
    Bone / Socket Index=$Particles.SpawnGroup
    Bone Sampling Mode=::NewEnumerator4
    Inherit Velocity=false
    Interpolate Mesh Sampling=false
    MaskSpawnGroup=false
    Sampled Position Offset=&MaskVectorBySpawnGroup(Masked Vector=<9 -2 -1>, Passthrough Vector=<-9 2 1>)
    Sampled Position Offset Coordinate Space:ENiagaraMeshTransforms=::NewEnumerator3
    Skeletal Mesh=$User.SkeletalMesh
   )
   &OrientMeshToVector(
    Look At Direction:Vector3f=&NormalizeVector(
     Vector To Normalize=&FN_FrontEndVector3Modifier(
      Gameplay=&TransformVector(
       Vector=&Multiply_Vector(
        A=&NormalizeVector(
         Vector To Normalize=&TransformVector(Vector=$Engine.Owner.Velocity, Source Space=::World, Destination Space=::Local)
         Safe Normalize=true
         Safe Fallback Threshold=0
         Safe Fallback Vector=<0 0 0.001>
        )
        B=<0 1 1>
       )
       Destination Space=::World
      )
      Front End=&TransformVector(Vector=&Lerp_Vector(A=<0 0.4 -1>, B=<0 2 -1>, Alpha=$User.ParaGlideLeanAlpha), Destination Space=::Simulation)
     )
     Safe Normalize=true
     Safe Fallback Threshold=0
     Safe Fallback Vector=<0 0 0.001>
    )
    Override Quat=false
    Side Axis Direction:Vector3f=$Engine.Owner.SystemYAxis
    UseOptionalSideAxis=true
    UseOptionalUpAxis=false
   )
   /SetVariables(Particles.GlowPosition:Vector3f=&Add_Vector(A=Linked() {linkedVariable {name=Particles.Position, type=NiagaraPosition}}, B=$System.TireOffset)) {isSetParameters=true}
   &CameraOffset(Camera Offset Amount=&FN_FrontEndFloatModifier(Gameplay=0, Front End=3))
   &Color(Color=&MakeLinearColorFromVectorAndFloat(
     "Vector (RGB)"=<0 111 111>
     "Float (Alpha)"=&FNTimeOfDayScalar(Night Scale=0.4, Morning Scale=0.5, Evening Scale=0.4)
    ))
  ]
  rendererClasses [
   &NiagaraMeshRendererProperties
   &NiagaraSpriteRendererProperties
   &NiagaraLightRendererProperties
  ]
  renderers [
   {rendererClass=&NiagaraMeshRendererProperties, rendererIndex=0}
   {rendererClass=&NiagaraSpriteRendererProperties, rendererIndex=1}
   {rendererClass=&NiagaraLightRendererProperties, rendererIndex=2}
  ]
 }
}
summary {
 systemName=NS_Contrail_Motorcyclist_Female
 emitters [
  {
   emitterName=Tires
   rendererClasses [
    &NiagaraMeshRendererProperties
    &NiagaraSpriteRendererProperties
    &NiagaraLightRendererProperties
   ]
  }
  {emitterName=Road_Ribbons, rendererClasses [&NiagaraRibbonRendererProperties]}
  {emitterName=Chains, rendererClasses [&NiagaraMeshRendererProperties]}
  {emitterName=Spark_Ribbon_Guides, rendererClasses []}
  {emitterName=Spark_Ribbons, rendererClasses [&NiagaraRibbonRendererProperties]}
  {emitterName=Spark_Sprites, rendererClasses [&NiagaraSpriteRendererProperties]}
  {bEnabled=false, emitterName=FN_Debug, rendererClasses [&NiagaraSpriteRendererProperties]}
 ]
 userVariables [
  User.bIsFrontEnd {type=NiagaraBool}
  User.bIsFrontEndPreview {type=NiagaraBool}
  User.GradientCenterpointSparksFlat {type=NiagaraFloat, defaultValue=0.5, description=""}
  User.GradientCenterpointSparksSide {type=NiagaraFloat, defaultValue=0.9, description=""}
  User.ParaGlideLeanAlpha {type=NiagaraFloat, defaultValue=0, description=""}
  User.PlayerSpeed {type=NiagaraFloat, defaultValue=0, description=""}
  User.RibbonIDOffset {type=NiagaraInt32}
  User.SkeletalMesh {
   type=NiagaraDataInterfaceSkeletalMesh
   defaultValue {
    struct=NiagaraExt_VariableValue_DataInterface
    value {dataInterface=&self:NiagaraDataInterfaceSkeletalMesh_0}
   }
  }
 ]
}
